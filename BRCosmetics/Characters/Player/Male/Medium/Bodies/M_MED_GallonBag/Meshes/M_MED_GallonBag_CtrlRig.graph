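# Control Rig for the M_MED_GallonBag male body that procedurally drives its secondary deform bones.
# It chains spine twist cancellation, mutually aimed sternum-to-hip pistons, shoulder and calf twist extraction, then wire-bone constraints.
# Swing-twist decomposition isolates twist, while spherical pose readers weight corrective offsets on the left clavicle and leg and hip wires.
# The Leg IK function is disconnected, as are the inner-clavicle aim constraints.
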ControlRigBlueprint /BRCosmetics/Characters/Player/Male/Medium/Bodies/M_MED_GallonBag/Meshes/M_MED_GallonBag_CtrlRig
graphs [
 New Function_ContainedGraph {nodes [
   RigVMFunctionEntryNode()
   > &RigUnit_AimConstraintLocalSpaceOffset(
    Child=(Type=Bone,Name="deform_sternum_piston_aimat_l")
    Filter=(bX=True,bY=True,bZ=True)
    Parents=((Item=(Type=Bone,Name="deform_hip_piston_aimat_l_02"),Weight=1))
    Weight=1
    WorldUp=(Target=(X=0,Y=0,Z=1),Kind=Location,Space=(Type=Bone,Name="deform_hip_piston_aimat_l_02"))
   )
   > &RigUnit_AimConstraintLocalSpaceOffset(
    Child=(Type=Bone,Name="deform_hip_piston_aimat_l_01")
    Filter=(bX=false,bY=true,bZ=false)
    Parents=((Item=(Type=Bone,Name="deform_sternum_piston_aimat_l"),Weight=1))
    Weight=1
    WorldUp=(Target=(X=0,Y=0,Z=1),Kind=Location,Space=(Type=Bone,Name="deform_sternum_piston_aimat_l"))
   )
   > &RigUnit_AimConstraintLocalSpaceOffset(
    Child=(Type=Bone,Name="deform_hip_piston_aimat_l_02")
    Filter=(bX=false,bY=true,bZ=true)
    Parents=((Item=(Type=Bone,Name="deform_sternum_piston_aimat_l"),Weight=1))
    Weight=1
    WorldUp=(Target=(X=0,Y=0,Z=1),Kind=Location,Space=(Type=Bone,Name="deform_sternum_piston_aimat_l"))
   )
   > &RigUnit_AimConstraintLocalSpaceOffset(
    Child=(Type=Bone,Name="deform_sternum_piston_aimat_r")
    Filter=(bX=True,bY=True,bZ=True)
    Parents=((Item=(Type=Bone,Name="deform_hip_piston_aimat_r_02"),Weight=1))
    Weight=1
    WorldUp=(Target=(X=0,Y=0,Z=1),Kind=Location,Space=(Type=Bone,Name="deform_hip_piston_aimat_r_02"))
   )
   > &RigUnit_AimConstraintLocalSpaceOffset(
    Child=(Type=Bone,Name="deform_hip_piston_aimat_r_01")
    Filter=(bX=false,bY=true,bZ=false)
    Parents=((Item=(Type=Bone,Name="deform_sternum_piston_aimat_r"),Weight=1))
    Weight=1
    WorldUp=(Target=(X=0,Y=0,Z=1),Kind=Location,Space=(Type=Bone,Name="deform_sternum_piston_aimat_r"))
   )
   > &RigUnit_AimConstraintLocalSpaceOffset(
    Child=(Type=Bone,Name="deform_hip_piston_aimat_r_02")
    Filter=(bX=false,bY=true,bZ=true)
    Parents=((Item=(Type=Bone,Name="deform_sternum_piston_aimat_r"),Weight=1))
    Weight=1
    WorldUp=(Target=(X=0,Y=0,Z=1),Kind=Location,Space=(Type=Bone,Name="deform_sternum_piston_aimat_r"))
   )
   > RigVMFunctionReturnNode()
  ]}
 New Function_ContainedGraph {nodes [
   &RigUnit_AimConstraintLocalSpaceOffset(
    Child=(Type=Bone,Name="deform_inner_clavicle_l")
    ExecutePin=()
    Filter=(bX=false,bY=true,bZ=True)
    Parents=((Item=(Type=Bone,Name="def_upperarm_l"),Weight=1))
    Weight=1
    WorldUp=(Target=(X=1,Y=0,Z=0),Kind=Direction,Space=(Type=Bone,Name="def_upperarm_l"))
    bMaintainOffset=false
   )
   > &RigUnit_AimConstraintLocalSpaceOffset(
    AimAxis=(X=-1,Y=0,Z=0)
    Child=(Type=Bone,Name="deform_inner_clavicle_r")
    Filter=(bX=false,bY=True,bZ=True)
    Parents=((Item=(Type=Bone,Name="def_upperarm_r"),Weight=1))
    Weight=1
    WorldUp=(Target=(X=-1,Y=0,Z=0),Kind=Direction,Space=(Type=Bone,Name="def_upperarm_r"))
    bMaintainOffset=false
   )
   RigVMFunctionEntryNode()
   > RigVMFunction_Sequence = RigVMAggregateNode()
   RigVMFunctionReturnNode(ExecuteContext=@RigVMFunction_Sequence.C)
   RigUnit_GetTransform = &RigUnit_GetTransform(Item=(Type=Bone,Name="def_upperarm_l"), Space=LocalSpace, bInitial=False)
   n1 = &RigVMFunction_MathQuaternionSwingTwist(TwistAxis=(X=1,Y=0,Z=0), Input=@RigUnit_GetTransform.Transform.Rotation)
   RigUnit_SetRotation = &RigUnit_SetRotation(Item=(Type=Bone,Name="def_upperarm_twist_l"), Weight=1, bInitial=False, Value=@n1.Twist, ExecutePin=@RigVMFunction_Sequence.B)
   > &RigUnit_RotationConstraintLocalSpaceOffset(
    AdvancedSettings=(InterpolationType=Shortest,RotationOrderForFilter=ZYX)
    Child=(Type=Bone,Name="deform_shoulder_l")
    Filter=(bX=true,bY=false,bZ=false)
    Parents=((Item=(Type=Bone,Name="def_upperarm_twist_l"),Weight=1))
    Weight=1
    bMaintainOffset=true
   )
   n2 = &RigVMFunction_MathQuaternionSwingTwist(
    TwistAxis=(X=0,Y=0,Z=0)
    Input=&RigUnit_GetTransform(Item=(Type=Bone,Name="def_upperarm_r"), Space=LocalSpace, bInitial=False).Transform.Rotation
   )
   > &RigUnit_SetRotation(Item=(Type=Bone,Name="def_upperarm_twist_r"), Weight=1, bInitial=False, Value=@n2.Twist)
   > &RigUnit_RotationConstraintLocalSpaceOffset(
    AdvancedSettings=(InterpolationType=Shortest,RotationOrderForFilter=ZYX)
    Child=(Type=Bone,Name="deform_shoulder_r")
    Filter=(bX=true,bY=false,bZ=false)
    Parents=((Item=(Type=Bone,Name="def_upperarm_twist_r"),Weight=1))
    Weight=1
    bMaintainOffset=true
   )
   n3 = &RigUnit_SphericalPoseReader(
    ActiveRegionScaleFactors=(PositiveWidth=1,NegativeWidth=1,PositiveHeight=1,NegativeHeight=1)
    ActiveRegionSize=0.1
    Debug=(bDrawDebug=true,bDraw2D=False,bDrawLocalAxes=False,DebugScale=25,DebugSegments=20,DebugThickness=0.25)
    DriverAxis=(X=1,Y=0,Z=0)
    DriverItem=(Type=Bone,Name="deform_clavicle_l")
    FalloffRegionScaleFactors=(PositiveWidth=1,NegativeWidth=1,PositiveHeight=1,NegativeHeight=1)
    FalloffSize=0.24
    FlipHeightScaling=False
    FlipWidthScaling=False
    OptionalParentItem=(Type=Bone,Name="None")
    RotationOffset=(X=0,Y=-45,Z=0)
    ExecutePin=@RigVMFunction_Sequence.A
   )
   > &RigUnit_ModifyTransforms(
    ItemToModify=((Item=(Type=Bone,Name="deform_clavicle_offset_l"),Transform=(Rotation=(X=0,Y=0.0436,Z=0,W=0.999),Translation=(X=0,Y=0,Z=0),Scale3D=(X=1,Y=1,Z=1))))
    WeightMinimum=0
    Weight=@n3.OutputParam
   )
   /Script/RigVM.RigVMFunction_MathQuaternionInverse(Value=@n1.Twist) {is=struct RigVMUnitNode, template="Inverse::Execute(in Value,out Result)"}
   &RigVMFunction_MathQuaternionInverse(Value=@n2.Twist) {template="Inverse::Execute(in Value,out Result)"}
   &RigUnit_ModifyTransforms(
    ExecutePin=()
    ItemToModify=((Item=(Type=Bone,Name="None"),Transform=(Translation=(X=0,Y=0,Z=0),Scale3D=(X=1,Y=1,Z=1))))
    Weight=1
    WeightMinimum=0
   )
   &RigUnit_SphericalPoseReader(
    ActiveRegionScaleFactors=(PositiveWidth=1,NegativeWidth=1,PositiveHeight=1,NegativeHeight=1)
    ActiveRegionSize=0.1
    Debug=(bDrawDebug=True,bDraw2D=False,bDrawLocalAxes=False,DebugScale=25,DebugSegments=20,DebugThickness=0.25)
    DriverAxis=(X=1,Y=0,Z=0)
    DriverItem=(Type=Bone,Name="None")
    ExecutePin=()
    FalloffRegionScaleFactors=(PositiveWidth=1,NegativeWidth=1,PositiveHeight=1,NegativeHeight=1)
    FalloffSize=0.2
    FlipHeightScaling=False
    FlipWidthScaling=False
    OptionalParentItem=(Type=Bone,Name="None")
    RotationOffset=(X=90,Y=0,Z=0)
   )
  ]}
 New Function_ContainedGraph {nodes [
   RigVMFunctionEntryNode()
   n1 = &RigUnit_GetTransform(Item=(Type=Bone,Name="deform_spine_03"), Space=LocalSpace, bInitial=False)
   > &RigUnit_SetRotation(
    Item=(Type=Bone,Name="deform_spine_03_no_twist")
    Weight=1
    bInitial=False
    Value=&RigVMFunction_MathQuaternionInverse(Value=&RigVMFunction_MathQuaternionSwingTwist(TwistAxis=(X=-1,Y=0,Z=0), Input=@n1.Transform.Rotation).Twist) {template="Inverse::Execute(in Value,out Result)"}.Result
   )
   > &RigUnit_ParentConstraint(
    AdvancedSettings=(InterpolationType=Average,RotationOrderForFilter=XZY)
    Child=(Type=Bone,Name="deform_spine_01_tweak")
    Parents=((Item=(Type=Bone,Name="pelvis"),Weight=0.5),(Item=(Type=Bone,Name="deform_spine_02"),Weight=0.5))
    Weight=1
   )
   > RigVMFunctionReturnNode()
   RigUnit_GetTransform = &RigUnit_GetTransform(Item=(Type=Bone,Name="deform_spine_03"), Space=LocalSpace, bInitial=False)
   RigVMFunction_MathQuaternionSwingTwist = &RigVMFunction_MathQuaternionSwingTwist(TwistAxis=(X=-1,Y=0,Z=0), Input=@RigUnit_GetTransform.Transform.Rotation)
   Inverse = &RigVMFunction_MathQuaternionInverse(Value=@RigVMFunction_MathQuaternionSwingTwist.Twist) {template="Inverse::Execute(in Value,out Result)"}
   RigUnit_SetRotation = &RigUnit_SetRotation(ExecutePin=(), Item=(Type=Bone,Name="deform_spine_02_no_twist"), Weight=1, bInitial=False, Value=@Inverse.Result)
  ]}
 New Function_ContainedGraph {nodes [
   RigVMFunctionEntryNode()
   RigUnit_GetTransform = &RigUnit_GetTransform(Item=(Type=Bone,Name="deform_foot_l"), Space=LocalSpace, bInitial=False)
   RigVMFunction_MathQuaternionSwingTwist = &RigVMFunction_MathQuaternionSwingTwist(TwistAxis=(X=1,Y=0,Z=0), Input=@RigUnit_GetTransform.Transform.Rotation)
   > RigUnit_SetRotation = &RigUnit_SetRotation(Item=(Type=Bone,Name="deform_calf_twist_l"), Weight=1, bInitial=False, Value=@RigVMFunction_MathQuaternionSwingTwist.Twist)
   > &RigUnit_SetRotation(
    Item=(Type=Bone,Name="deform_calf_twist_r")
    Weight=1
    bInitial=False
    Value=&RigVMFunction_MathQuaternionSwingTwist(
     TwistAxis=(X=1,Y=0,Z=0)
     Input=&RigUnit_GetTransform(Item=(Type=Bone,Name="deform_foot_r"), Space=LocalSpace, bInitial=False).Transform.Rotation
    ).Twist
   )
   > RigVMFunctionReturnNode()
  ]}
 New Function_ContainedGraph {nodes [
   RigVMFunctionEntryNode()
   > RigVMFunctionReturnNode()
   RigUnit_GetTransform = &RigUnit_GetTransform(Item=(Type=Bone,Name="foot_l"), bInitial=false)
   /Script/ControlRig.RigUnit_CCDIKItemArray(
    BaseRotationLimit=30
    ExecutePin=()
    Items=((Type=Bone,Name="deform_thigh_l"),(Type=Bone,Name="deform_thigh_knee_l"),(Type=Bone,Name="deform_calf_l"),(Type=Bone,Name="deform_foot_l"))
    MaxIterations=4
    Precision=1
    RotationLimits=()
    Weight=1
    bPropagateToChildren=true
    bStartFromTail=false
    EffectorTransform=@RigUnit_GetTransform.Transform
   ) {is=struct RigVMUnitNode}
   /Script/ControlRig.RigUnit_FABRIKItemArray(
    ExecutePin=()
    Items=((Type=Bone,Name="deform_thigh_l"),(Type=Bone,Name="deform_thigh_knee_l"),(Type=Bone,Name="deform_calf_l"),(Type=Bone,Name="deform_foot_l"))
    MaxIterations=10
    Precision=1
    Weight=1
    bPropagateToChildren=true
    bSetEffectorTransform=True
    EffectorTransform=@RigUnit_GetTransform.Transform
   ) {is=struct RigVMUnitNode}
  ]}
 New Function_ContainedGraph {nodes [
   RigVMFunctionEntryNode()
   > &RigUnit_ParentConstraint(
    AdvancedSettings=(InterpolationType=Average,RotationOrderForFilter=XZY)
    Child=(Type=Bone,Name="dyn_l_leg_wire_04")
    Parents=((Item=(Type=Bone,Name="deform_thigh_l"),Weight=0.5),(Item=(Type=Bone,Name="deform_calf_l"),Weight=0.5))
    Weight=1
   )
   > &RigUnit_ParentConstraint(
    AdvancedSettings=(InterpolationType=Average,RotationOrderForFilter=XZY)
    Child=(Type=Bone,Name="dyn_l_leg_wire_01")
    Parents=((Item=(Type=Bone,Name="deform_thigh_l"),Weight=0.8),(Item=(Type=Bone,Name="deform_calf_l"),Weight=0.2))
    Weight=1
   )
   > &RigUnit_ParentConstraint(
    AdvancedSettings=(InterpolationType=Average,RotationOrderForFilter=XZY)
    Child=(Type=Bone,Name="dyn_r_leg_wire_04")
    Parents=((Item=(Type=Bone,Name="deform_thigh_r"),Weight=0.2),(Item=(Type=Bone,Name="deform_calf_r"),Weight=0.8))
    Weight=1
   )
   > &RigUnit_ParentConstraint(
    AdvancedSettings=(InterpolationType=Average,RotationOrderForFilter=XZY)
    Child=(Type=Bone,Name="dyn_r_leg_wire_05")
    Parents=((Item=(Type=Bone,Name="deform_thigh_r"),Weight=0.5),(Item=(Type=Bone,Name="deform_calf_r"),Weight=0.5))
    Weight=1
   )
   > &RigUnit_ParentConstraint(
    AdvancedSettings=(InterpolationType=Average,RotationOrderForFilter=XZY)
    Child=(Type=Bone,Name="dyn_l_hip_wire_01")
    Parents=((Item=(Type=Bone,Name="deform_thigh_l"),Weight=0.5),(Item=(Type=Bone,Name="pelvis"),Weight=0.5))
    Weight=1
   )
   > &RigUnit_ParentConstraint(
    AdvancedSettings=(InterpolationType=Average,RotationOrderForFilter=XZY)
    Child=(Type=Bone,Name="dyn_l_hip_wire_02")
    Parents=((Item=(Type=Bone,Name="deform_thigh_l"),Weight=0.8),(Item=(Type=Bone,Name="pelvis"),Weight=0.2))
    Weight=1
   )
   > &RigUnit_ParentConstraint(
    AdvancedSettings=(InterpolationType=Average,RotationOrderForFilter=XZY)
    Child=(Type=Bone,Name="dyn_l_rear_wire_02")
    Parents=((Item=(Type=Bone,Name="deform_spine_02"),Weight=0.5),(Item=(Type=Bone,Name="deform_spine_03_no_twist"),Weight=0.5))
    Weight=1
   )
   > &RigUnit_ParentConstraint(
    AdvancedSettings=(InterpolationType=Average,RotationOrderForFilter=XZY)
    Child=(Type=Bone,Name="dyn_r_rear_wire_02")
    Parents=((Item=(Type=Bone,Name="deform_spine_02"),Weight=0.5),(Item=(Type=Bone,Name="deform_spine_03_no_twist"),Weight=0.5))
    Weight=1
   )
   > n1 = &RigUnit_SphericalPoseReader(
    ActiveRegionScaleFactors=(PositiveWidth=3.5,NegativeWidth=0.5,PositiveHeight=3,NegativeHeight=3)
    ActiveRegionSize=0.1
    Debug=(bDrawDebug=false,DebugScale=25,DebugSegments=20,DebugThickness=0.25)
    DriverAxis=(X=-1,Y=0,Z=0)
    DriverItem=(Type=Bone,Name="deform_thigh_l")
    FalloffRegionScaleFactors=(PositiveWidth=1,NegativeWidth=1,PositiveHeight=1,NegativeHeight=1)
    FalloffSize=0.45
    OptionalParentItem=(Type=Bone)
    RotationOffset=(X=90,Y=0,Z=0)
   )
   > &RigUnit_ModifyTransforms(
    ItemToModify="((Item=(Type=Bone,Name=\"dyn_l_hip_wire_01\"),Transform=(Rotation=(X=-0.5018,Y=-0.169,Z=-0.3522,W=0.7718),Translation=(X=0.2393,Y=2.4601,Z=0.6969),Scale3D=(X=1,Y=1,Z=1))),(Item=(Type=Bone,Name=\"dyn_l_hip_wire_02\"),Transform=(Rotation=(X=0,Y=0,Z=0,W=1),Transl ...(truncated)"
    WeightMinimum=0
    Weight=@n1.OutputParam
   )
   > n2 = &RigUnit_SphericalPoseReader(
    ActiveRegionScaleFactors=(PositiveWidth=1,NegativeWidth=1,PositiveHeight=1,NegativeHeight=1)
    ActiveRegionSize=0.1
    Debug=(bDrawDebug=True,bDraw2D=False,bDrawLocalAxes=False,DebugScale=25,DebugSegments=20,DebugThickness=0.25)
    DriverAxis=(X=-1,Y=0,Z=0)
    DriverItem=(Type=Bone,Name="deform_calf_l")
    FalloffRegionScaleFactors=(PositiveWidth=1,NegativeWidth=1,PositiveHeight=1,NegativeHeight=1)
    FalloffSize=0.5
    FlipHeightScaling=False
    FlipWidthScaling=False
    OptionalParentItem=(Type=Bone,Name="deform_thigh_l")
    RotationOffset=(X=0,Y=90,Z=130)
   )
   > &RigUnit_ModifyTransforms(
    ItemToModify="((Item=(Type=Bone,Name=\"dyn_l_leg_wire_01\"),Transform=(Rotation=(X=-0.0776,Y=0.202,Z=0.3501,W=0.9114),Translation=(X=-1.1676,Y=-1.0522,Z=0.9224),Scale3D=(X=1,Y=1,Z=1))),(Item=(Type=Bone,Name=\"dyn_l_leg_wire_04\"),Transform=(Rotation=(X=0,Y=0,Z=0.2753,W=0.96 ...(truncated)"
    WeightMinimum=0
    Weight=@n2.OutputParam
   )
   > n3 = &RigUnit_SphericalPoseReader(
    ActiveRegionScaleFactors=(PositiveWidth=1,NegativeWidth=1,PositiveHeight=1,NegativeHeight=1)
    ActiveRegionSize=0.1
    Debug=(bDrawDebug=True,bDraw2D=False,bDrawLocalAxes=False,DebugScale=25,DebugSegments=20,DebugThickness=0.25)
    DriverAxis=(X=1,Y=0,Z=0)
    DriverItem=(Type=Bone,Name="deform_calf_r")
    FalloffRegionScaleFactors=(PositiveWidth=1,NegativeWidth=1,PositiveHeight=1,NegativeHeight=1)
    FalloffSize=0.6
    FlipHeightScaling=False
    FlipWidthScaling=False
    OptionalParentItem=(Type=Bone,Name="deform_thigh_r")
    RotationOffset=(X=0,Y=90,Z=-30)
   )
   > &RigUnit_ModifyTransforms(
    ItemToModify="((Item=(Type=Bone,Name=\"dyn_r_leg_wire_01\"),Transform=(Rotation=(X=0,Y=0,Z=0.7463,W=0.6656),Translation=(X=-10.1135,Y=-0.452,Z=0),Scale3D=(X=1,Y=1,Z=1))),(Item=(Type=Bone,Name=\"dyn_r_leg_wire_05\"),Transform=(Rotation=(X=0,Y=0,Z=0,W=1),Translation=(X=-1.778 ...(truncated)"
    WeightMinimum=0
    Weight=@n3.OutputParam
   )
   > RigVMFunctionReturnNode()
  ]}
 RigVMFunctionLibrary {nodes [
   Ankle Twist = RigVMCollapseNode(ExecuteContext=())
   Leg IK = RigVMCollapseNode(ExecuteContext=())
   Piston Aims = RigVMCollapseNode(ExecuteContext=())
   Shoulder Circle = RigVMCollapseNode(ExecuteContext=())
   Spine Twist = RigVMCollapseNode(ExecuteContext=())
   Wire Constraints = RigVMCollapseNode(ExecuteContext=())
  ]}
 RigVMFunction_Sequence_1_ContainedGraph {nodes [
   n1 = RigVMFunctionEntryNode()
   RigVMFunction_Sequence = &RigVMFunction_Sequence(ExecuteContext=@n1.ExecuteContext)
   n2 = &RigVMFunction_Sequence(ExecuteContext=@RigVMFunction_Sequence.B)
   RigVMFunctionReturnNode(A=@RigVMFunction_Sequence.A, B=@n2.A, C=@n2.B)
  ]}
 RigVMModel {nodes [
   RigUnit_BeginExecution = &RigUnit_BeginExecution()
   Spine Twist = RigVMFunctionReferenceNode(ExecuteContext=@RigUnit_BeginExecution.ExecutePin)
   > Piston Aims = RigVMFunctionReferenceNode()
   > Shoulder Circle = RigVMFunctionReferenceNode()
   > Ankle Twist = RigVMFunctionReferenceNode()
   > Wire Constraints = RigVMFunctionReferenceNode()
   Leg IK = RigVMFunctionReferenceNode(ExecuteContext=())
  ]}
]
hierarchy [
 root {loc=<0 0 0>, rot=<0 0 0 1>}
 pelvis {parent=root, loc=<0 -0.7071 93.1191>, rot=<0.0127 0.707 -0.0127 -0.707>}
 spine_01 {parent=pelvis, loc=<6.7494 -0.6388 0>, rot=<0 0 0.034 -0.9994>}
 spine_02 {parent=spine_01, loc=<7.2518 -0.035 0>, rot=<0 0 0.0238 -0.9997>}
 spine_03 {parent=spine_02, loc=<6.0104 0 0>, rot=<0 0 0.0149 -0.9999>}
 spine_04 {parent=spine_03, loc=<8.1024 0 0>, rot=<0 0 -0.0132 -0.9999>}
 spine_05 {parent=spine_04, loc=<10.0146 0 0>, rot=<0 0 -0.0015 -1>}
 attach_cape {parent=spine_05, loc=<9.017 -11.714 0>, rot=<0.0616 -0.7044 -0.0616 -0.7044>}
 neck_01 {parent=spine_05, loc=<11.3138 0.1529 0>, rot=<0 0 -0.2003 -0.9797>}
 neck_02 {parent=neck_01, loc=<4.5627 0 0>, rot=<0 0 0.0504 -0.9987>}
 head {parent=neck_02, loc=<5.3914 0 0>, rot=<0 0 0.115 -0.9934>}
 attach_backpack {parent=spine_05, loc=<-0.123 -13.473 0>, rot=<-0.0093 -0.707 0.0093 -0.707>}
 clavicle_l {parent=spine_05, loc=<4.3575 -0.1466 -2.3223>, rot=<0.0465 -0.7289 -0.009 -0.6829>}
 upperarm_l {parent=clavicle_l, loc=<15.5111 0 0>, rot=<0.0226 -0.3234 -0.0088 -0.9459>}
 lowerarm_l {parent=upperarm_l, loc=<26.6514 0 0>, rot=<0.0391 -0.0099 -0.2453 -0.9686>}
 hand_l {parent=lowerarm_l, loc=<22.3491 0 0>, rot=<-0.7081 -0.0507 0.0975 -0.6975>}
 thumb_01_l {parent=hand_l, loc=<1.3118 -0.3191 -2.459>, rot=<0.6546 -0.4042 0.0664 -0.6354>}
 thumb_02_l {parent=thumb_01_l, loc=<5.9336 -0.0448 0.0401>, rot=<-0.0037 0.0213 0.298 -0.9543>}
 thumb_03_l {parent=thumb_02_l, loc=<2.2223 0.3682 -0.0907>, rot=<0.0118 -0.013 0.0066 -0.9998>}
 index_metacarpal_l {parent=hand_l, loc=<1.6535 0.0183 -2.96>, rot=<0.0158 -0.0124 -0.0173 -0.9996>}
 index_01_l {parent=index_metacarpal_l, loc=<8.3008 0.0357 -0.0524>, rot=<-0.0044 0.0485 0.0924 -0.9945>}
 index_02_l {parent=index_01_l, loc=<3.3587 -0.3004 0.0452>, rot=<-0.002 0.0027 0.3225 -0.9466>}
 index_03_l {parent=index_02_l, loc=<2.5578 -0.123 -0.0252>, rot=<-0.002 0.0067 0.2558 -0.9667>}
 middle_metacarpal_l {parent=hand_l, loc=<2.1856 -0.1725 -0.5984>, rot=<0.006 0.0168 -0.0329 -0.9993>}
 middle_01_l {parent=middle_metacarpal_l, loc=<7.3275 -0.2447 0.0129>, rot=<-0.0208 0.0042 0.0858 -0.9961>}
 middle_02_l {parent=middle_01_l, loc=<4.0123 0.1401 0.1175>, rot=<-0.0218 0.0258 0.3512 -0.9357>}
 middle_03_l {parent=middle_02_l, loc=<2.6077 0.0534 0.1373>, rot=<0.0082 0.0188 0.1708 -0.9851>}
 ring_metacarpal_l {parent=hand_l, loc=<2.2074 -0.3625 1.3638>, rot=<-0.0284 0.0722 -0.0458 -0.9959>}
 ring_01_l {parent=ring_metacarpal_l, loc=<7.0295 -0.6197 -0.1446>, rot=<-0.0489 -0.0252 0.1892 -0.9804>}
 ring_02_l {parent=ring_01_l, loc=<3.6174 0.4647 -0.1027>, rot=<0 0 0.2806 -0.9598>}
 ring_03_l {parent=ring_02_l, loc=<2.4152 0.114 0.0199>, rot=<0 0 0.1189 -0.9929>}
 pinky_metacarpal_l {parent=hand_l, loc=<1.7616 -0.9753 3.3814>, rot=<-0.0397 0.0814 -0.069 -0.9935>}
 pinky_01_l {parent=pinky_metacarpal_l, loc=<7.1709 -0.4831 -0.1094>, rot=<-0.0654 -0.0035 0.2242 -0.9723>}
 pinky_02_l {parent=pinky_01_l, loc=<2.8274 0.0893 -0.1043>, rot=<0.0037 0.0011 0.2854 -0.9584>}
 pinky_03_l {parent=pinky_02_l, loc=<2.2046 0.0307 -0.0078>, rot=<-0.0002 0.0098 0.1445 -0.9895>}
 lowerarm_twist_02_l {parent=lowerarm_l, loc=<9.6199 0 0>, rot=<0 0 0 -1>}
 lowerarm_twist_01_l {parent=lowerarm_l, loc=<17.8549 0 0>, rot=<0 0 0 -1>}
 upperarm_twist_01_l {parent=upperarm_l, loc=<0.0041 0 0>, rot=<0 0 0 1>}
 upperarm_twist_02_l {parent=upperarm_l, loc=<13.7941 0 0>, rot=<0 0 0 1>}
 clavicle_r {parent=spine_05, loc=<4.3575 -0.1466 2.3223>, rot=<0.7289 0.0465 -0.6829 0.009>}
 upperarm_r {parent=clavicle_r, loc=<-15.5111 0 0>, rot=<0.0226 -0.3234 -0.0088 -0.9459>}
 lowerarm_r {parent=upperarm_r, loc=<-26.6514 0 0>, rot=<0.0391 -0.0099 -0.2453 -0.9686>}
 hand_r {parent=lowerarm_r, loc=<-22.3491 0 0>, rot=<-0.7081 -0.0507 0.0975 -0.6975>}
 weapon_r {parent=hand_r, loc=<0.1974 3.4111 -0.3811>, rot=<0 0 0 1>}
 thumb_01_r {parent=hand_r, loc=<-1.3118 0.3191 2.459>, rot=<0.6546 -0.4042 0.0664 -0.6354>}
 thumb_02_r {parent=thumb_01_r, loc=<-5.9336 0.0448 -0.0401>, rot=<-0.0037 0.0213 0.298 -0.9543>}
 thumb_03_r {parent=thumb_02_r, loc=<-2.2223 -0.3682 0.0907>, rot=<0.0118 -0.013 0.0066 -0.9998>}
 index_metacarpal_r {parent=hand_r, loc=<-1.6535 -0.0183 2.96>, rot=<0.0158 -0.0124 -0.0173 -0.9996>}
 index_01_r {parent=index_metacarpal_r, loc=<-8.3008 -0.0357 0.0524>, rot=<-0.0044 0.0485 0.0924 -0.9945>}
 index_02_r {parent=index_01_r, loc=<-3.3587 0.3004 -0.0452>, rot=<-0.002 0.0027 0.3225 -0.9466>}
 index_03_r {parent=index_02_r, loc=<-2.5578 0.123 0.0252>, rot=<-0.002 0.0067 0.2558 -0.9667>}
 middle_metacarpal_r {parent=hand_r, loc=<-2.1856 0.1725 0.5984>, rot=<0.006 0.0168 -0.0329 -0.9993>}
 middle_01_r {parent=middle_metacarpal_r, loc=<-7.3275 0.2447 -0.0129>, rot=<-0.0208 0.0042 0.0858 -0.9961>}
 middle_02_r {parent=middle_01_r, loc=<-4.0123 -0.1401 -0.1175>, rot=<-0.0218 0.0258 0.3512 -0.9357>}
 middle_03_r {parent=middle_02_r, loc=<-2.6077 -0.0534 -0.1373>, rot=<0.0082 0.0188 0.1708 -0.9851>}
 ring_metacarpal_r {parent=hand_r, loc=<-2.2074 0.3625 -1.3638>, rot=<-0.0284 0.0722 -0.0458 -0.9959>}
 ring_01_r {parent=ring_metacarpal_r, loc=<-7.0295 0.6197 0.1446>, rot=<-0.0489 -0.0252 0.1892 -0.9804>}
 ring_02_r {parent=ring_01_r, loc=<-3.6174 -0.4647 0.1027>, rot=<0 0 0.2806 -0.9598>}
 ring_03_r {parent=ring_02_r, loc=<-2.4152 -0.114 -0.0199>, rot=<0 0 0.1189 -0.9929>}
 pinky_metacarpal_r {parent=hand_r, loc=<-1.7616 0.9753 -3.3814>, rot=<-0.0397 0.0814 -0.069 -0.9935>}
 pinky_01_r {parent=pinky_metacarpal_r, loc=<-7.1709 0.4831 0.1094>, rot=<-0.0654 -0.0035 0.2242 -0.9723>}
 pinky_02_r {parent=pinky_01_r, loc=<-2.8274 -0.0893 0.1043>, rot=<0.0037 0.0011 0.2854 -0.9584>}
 pinky_03_r {parent=pinky_02_r, loc=<-2.2046 -0.0307 0.0078>, rot=<-0.0002 0.0098 0.1445 -0.9895>}
 lowerarm_twist_02_r {parent=lowerarm_r, loc=<-9.6199 0 0>, rot=<0 0 0 -1>}
 lowerarm_twist_01_r {parent=lowerarm_r, loc=<-17.8549 0 0>, rot=<0 0 0 -1>}
 upperarm_twist_01_r {parent=upperarm_r, loc=<-0.0041 0 0>, rot=<0 0 0 -1>}
 upperarm_twist_02_r {parent=upperarm_r, loc=<-13.7941 0 0>, rot=<0 0 0 -1>}
 thigh_l {parent=pelvis, loc=<-2.5693 -0.0109 -7.7501>, rot=<0.0012 0.0572 0.0155 -0.9982>}
 calf_l {parent=thigh_l, loc=<-39.0484 0 0>, rot=<0.0006 -0.0161 -0.0391 -0.9991>}
 foot_l {parent=calf_l, loc=<-42.561 0.1077 0.4292>, rot=<0.0418 -0.0455 0.0343 -0.9975>}
 ball_l {parent=foot_l, loc=<-6.1599 13.8264 0.6806>, rot=<0.0001 -0.0001 -0.7063 -0.7079>}
 calf_twist_02_l {parent=calf_l, loc=<-18.8064 -0.0967 0.0397>, rot=<0.0237 0.0095 -0.0035 -0.9997>}
 calf_twist_01_l {parent=calf_l, loc=<-35.813 -0.0967 0.0397>, rot=<0.0237 0.0095 -0.0035 -0.9997>}
 thigh_twist_01_l {parent=thigh_l, loc=<-21.6303 0 0>, rot=<0.0013 0 0 -1>}
 thigh_r {parent=pelvis, loc=<-2.5693 -0.0109 7.7501>, rot=<0.0572 -0.0012 0.9982 0.0155>}
 calf_r {parent=thigh_r, loc=<39.0484 0 0>, rot=<0.0006 -0.0161 -0.0391 -0.9991>}
 foot_r {parent=calf_r, loc=<42.561 -0.1077 -0.4292>, rot=<0.0418 -0.0455 0.0343 -0.9975>}
 ball_r {parent=foot_r, loc=<6.1599 -13.8264 -0.6806>, rot=<0.0001 -0.0001 -0.7063 -0.7079>}
 calf_twist_02_r {parent=calf_r, loc=<18.8064 0.0967 -0.0397>, rot=<0.0237 0.0095 -0.0035 -0.9997>}
 calf_twist_01_r {parent=calf_r, loc=<35.813 0.0967 -0.0397>, rot=<0.0237 0.0095 -0.0035 -0.9997>}
 thigh_twist_01_r {parent=thigh_r, loc=<21.6303 0 0>, rot=<0.0013 0 0 -1>}
 deform_spine_01 {parent=pelvis, loc=<1.5165 0 0>, rot=<0 0 0.034 -0.9994>}
 deform_spine_02 {parent=deform_spine_01, loc=<8.3267 -0.1492 0>, rot=<0 0 0.0238 -0.9997>}
 deform_spine_03 {parent=deform_spine_02, loc=<6.038 0.0264 0>, rot=<0 0 0.0149 -0.9999>}
 deform_spine_04 {parent=deform_spine_03, loc=<7.6524 0 0>, rot=<0 0 -0.0132 -0.9999>}
 deform_spine_05 {parent=deform_spine_04, loc=<14.6293 0.0082 0>, rot=<0 0 -0.0015 -1>}
 deform_clavicle_l {parent=deform_spine_05, loc=<4.3575 -0.1466 -2.3223>, rot=<0.0465 -0.7289 -0.009 -0.6829>}
 def_upperarm_l {parent=deform_clavicle_l, loc=<17.9459 0.1205 -1.8834>, rot=<0.0137 -0.3228 0.0005 -0.9464>}
 def_upperarm_elbow_l {parent=def_upperarm_l, loc=<19.8165 0 0>, rot=<0.0372 -0.0081 -0.0971 -0.9946>}
 def_lowerarm_l {parent=def_upperarm_elbow_l, loc=<5.106 0 0>, rot=<0.0603 0.0063 -0.1531 -0.9864>}
 def_hand_l {parent=def_lowerarm_l, loc=<20.855 0 0>, rot=<-0.7395 -0.0568 0.0931 -0.6643>}
 def_upperarm_twist_l {parent=def_upperarm_l, loc=<0 0 0>, rot=<-0.4254 -0.3445 0.0357 -0.8361>}
 def_upperarm_l_ctrl {controlType=EULER_TRANSFORM, parent=def_upperarm_twist_l, type=Control, loc=<0 0 0>, rot=<0 0 0 1>}
 deform_inner_clavicle_l {parent=deform_clavicle_l, loc=<11.4973 0.0224 -0.2257>, rot=<0.0008 0.0336 -0.0252 -0.9991>}
 deform_clavicle_offset_l {parent=deform_clavicle_l, loc=<0 0 0>, rot=<0 0 0 1>}
 deform_shoulder_l {parent=deform_clavicle_offset_l, loc=<14.914 0.1841 -0.0872>, rot=<0.0008 0.0336 -0.0252 -0.9991>}
 deform_clavicle_offset_l_ctrl {controlType=EULER_TRANSFORM, parent=deform_clavicle_offset_l, type=Control, loc=<0 0 0>, rot=<0 0 0 1>}
 deform_clavicle_r {parent=deform_spine_05, loc=<4.3578 -0.1466 2.3223>, rot=<0.7289 0.0465 -0.6829 0.009>}
 def_upperarm_r {parent=deform_clavicle_r, loc=<-17.946 -0.1205 1.8833>, rot=<0.0138 -0.3227 0.0004 -0.9464>}
 def_upperarm_elbow_r {parent=def_upperarm_r, loc=<-19.8167 0 0.0003>, rot=<0.0372 -0.0081 -0.0971 -0.9946>}
 def_lowerarm_r {parent=def_upperarm_elbow_r, loc=<-5.1059 0 -0.0001>, rot=<0.0603 0.0063 -0.1531 -0.9864>}
 def_hand_r {parent=def_lowerarm_r, loc=<-20.8549 0 -0.0001>, rot=<-0.7359 -0.0573 0.0943 -0.668>}
 def_upperarm_twist_r {parent=def_upperarm_r, loc=<0 0 0>, rot=<0 0 0 1>}
 def_upperarm_r_ctrl {controlType=EULER_TRANSFORM, parent=def_upperarm_twist_r, type=Control, loc=<0 0 0>, rot=<0 0 0 1>}
 deform_inner_clavicle_r {parent=deform_clavicle_r, loc=<-11.4974 -0.0224 0.2262>, rot=<0.0008 0.0336 -0.0252 -0.9991>}
 deform_clavicle_offset_r {parent=deform_clavicle_r, loc=<0 0 0>, rot=<0 0 0 1>}
 deform_shoulder_r {parent=deform_clavicle_offset_r, loc=<-14.914 -0.1841 0.0877>, rot=<0.0008 0.0336 -0.0252 -0.9991>}
 deform_clavicle_offset_r_ctrl {controlType=EULER_TRANSFORM, parent=deform_clavicle_offset_r, type=Control, loc=<0 0 0>, rot=<0 0 0 1>}
 NewBone {parent=deform_spine_05, loc=<1.7851 -0.7121 -6.5015>, rot=<0.0635 -0.9956 -0.0123 0.0672>}
 deform_spine_03_no_twist {parent=deform_spine_03, loc=<0 0 0>, rot=<0 0 0 1>}
 deform_sternum_piston_aimat_l {parent=deform_spine_03_no_twist, loc=<3.3837 6.4557 -9.92>, rot=<0.0218 -0.8696 -0.4929 0.0188>}
 deform_sternum_piston_aimat_r {parent=deform_spine_03_no_twist, loc=<3.3835 6.4557 9.92>, rot=<0.8696 0.0218 0.0188 0.4929>}
 dyn_l_rear_wire_01 {parent=deform_spine_03_no_twist, loc=<8.6531 -7.122 -8.0318>, rot=<-0.1226 0.1844 0.3277 -0.9185>}
 dyn_l_rear_wire_02 {parent=deform_spine_03_no_twist, loc=<-0.7707 -10.8257 -7.2767>, rot=<-0.1019 0.3633 0.0366 -0.9254>}
 dyn_r_mid_front_wire {parent=deform_spine_03_no_twist, loc=<0.6113 8.4046 3.3491>, rot=<0 0.379 0 -0.9254>}
 dyn_l_mid_front_wire {parent=deform_spine_03_no_twist, loc=<1.6151 8.1772 -3.3289>, rot=<-0.074 -0.2324 0.0092 -0.9698>}
 dyn_r_rear_wire_02 {parent=deform_spine_03_no_twist, loc=<-0.7704 -10.8257 7.2767>, rot=<-0.1661 -0.1748 0.1759 -0.9544>}
 dyn_r_rear_wire_01 {parent=deform_spine_03_no_twist, loc=<8.6532 -6.6325 8.6238>, rot=<0.3016 0.0117 0.0881 -0.9493>}
 deform_spine_01_tweak {parent=deform_spine_01, loc=<0 0 0>, rot=<0 0 0 1>}
 deform_hip_piston_aimat_l_01 {parent=deform_spine_01_tweak, loc=<4.5535 5.2479 -9.2874>, rot=<-0.0224 -0.8651 -0.5008 -0.0175>}
 deform_hip_piston_aimat_l_02 {parent=deform_hip_piston_aimat_l_01, loc=<0 0 1.4201>, rot=<0 0 0 1>}
 deform_hip_piston_aimat_r_01 {parent=deform_spine_01_tweak, loc=<4.5535 5.2479 9.2874>, rot=<-0.8651 0.0224 0.0175 -0.5008>}
 deform_hip_piston_aimat_r_02 {parent=deform_hip_piston_aimat_r_01, loc=<0 0 -1.4201>, rot=<0 0 0 1>}
 deform_rear_cannister {parent=deform_spine_01_tweak, loc=<9.9617 -10.8657 0>, rot=<0 0 0 1>}
 dyn_mid_rear_wire {parent=deform_rear_cannister, loc=<12.6844 -0.5579 -0.1496>, rot=<0 0 0.3724 -0.9281>}
 deform_thigh_l {parent=pelvis, loc=<-3.7189 0.8288 -8.7986>, rot=<0.0292 0.0406 0.0068 -0.9987>}
 deform_thigh_knee_l {parent=deform_thigh_l, loc=<-34.655 0 0>, rot=<0 -0.0033 -0.0367 -0.9993>}
 deform_calf_l {parent=deform_thigh_knee_l, loc=<-5.3662 0.1798 0.0772>, rot=<0.0028 0.0043 0.0074 -1>}
 deform_calf_twist_l {parent=deform_calf_l, loc=<-32.3979 0.1665 -0.0307>, rot=<0 0 0 1>}
 deform_foot_l {parent=deform_calf_l, loc=<-40.3167 0.1287 -0.0874>, rot=<0.0126 -0.0457 0.033 -0.9983>}
 dyn_l_leg_wire_01 {parent=deform_thigh_l, loc=<-34.1193 1.2893 5.2077>, rot=<-0.2595 -0.0334 0.1231 -0.9573>}
 dyn_l_leg_wire_02 {parent=deform_thigh_l, loc=<-19.9554 -3.0365 3.6754>, rot=<-0.2774 -0.0761 0.1752 -0.9416>}
 dyn_l_leg_wire_03 {parent=deform_thigh_l, loc=<-14.1838 -2.3477 3.0874>, rot=<-0.1902 0.1413 -0.006 -0.9715>}
 dyn_l_leg_wire_04 {parent=deform_thigh_l, loc=<-38.2282 1.9803 -5.4346>, rot=<0.1677 0.0144 -0.0842 -0.9821>}
 dyn_l_leg_wire_05 {parent=deform_thigh_l, loc=<-22.2532 4.9233 -0.3577>, rot=<0.2603 0.1524 0.0363 -0.9527>}
 dyn_l_leg_wire_06 {parent=deform_thigh_l, loc=<-14.6868 2.2006 2.9645>, rot=<-0.4044 0.1539 -0.0273 -0.9011>}
 deform_thigh_r {parent=pelvis, loc=<-3.7188 0.8288 8.7986>, rot=<0.0406 -0.028 0.9988 0.0068>}
 deform_thigh_knee_r {parent=deform_thigh_r, loc=<34.6551 0 0>, rot=<0 -0.0033 -0.0367 -0.9993>}
 deform_calf_r {parent=deform_thigh_knee_r, loc=<5.3661 -0.1798 -0.0773>, rot=<0.0028 0.0043 0.0074 -1>}
 deform_calf_twist_r {parent=deform_calf_r, loc=<32.3979 -0.1665 0.0307>, rot=<0 0 0 -1>}
 deform_foot_r {parent=deform_calf_r, loc=<40.3167 -0.1287 0.0874>, rot=<0.0138 -0.0457 0.0331 -0.9983>}
 dyn_r_leg_wire_01 {parent=deform_thigh_r, loc=<33.5787 4.6834 -1.9006>, rot=<0.1681 0 0 -0.9858>}
 dyn_r_leg_wire_02 {parent=deform_thigh_r, loc=<18.5645 3.1734 -3.7581>, rot=<-0.4203 -0.0294 0.4356 -0.7955>}
 dyn_r_leg_wire_04 {parent=deform_thigh_r, loc=<39.1692 -2.4004 5.5373>, rot=<-0.1909 -0.1102 0.132 -0.9664>}
 dyn_r_leg_wire_05 {parent=deform_thigh_r, loc=<34.4281 0.3701 6.4728>, rot=<-0.0499 0.0317 0.3924 -0.9179>}
 dyn_r_leg_wire_06 {parent=deform_thigh_r, loc=<14.6869 -2.2006 -2.9645>, rot=<-0.3864 0.1659 0.0318 -0.9068>}
 dyn_r_leg_wire_03 {parent=deform_thigh_r, loc=<14.1838 2.3477 -3.0874>, rot=<-0.278 0.1055 -0.0951 -0.95>}
 dyn_r_leg_wire_07 {parent=deform_thigh_r, loc=<28.3302 -3.4587 -5.3825>, rot=<-0.1041 -0.0472 -0.1627 -0.98>}
 dyn_l_hip_wire_01 {parent=pelvis, loc=<-2.3554 7.5104 -8.498>, rot=<0.1242 0.3407 0.0264 -0.9316>}
 dyn_l_hip_wire_02 {parent=pelvis, loc=<-6.9888 8.712 -7.6184>, rot=<0.1031 -0.2833 0.0984 -0.9484>}
 ik_hand_root {parent=root, loc=<0 0 0>, rot=<0 0 0 1>}
 ik_hand_gun {parent=ik_hand_root, loc=<-51.9727 4.0883 102.6935>, rot=<0.6084 0.1793 0.4613 -0.6205>}
]
