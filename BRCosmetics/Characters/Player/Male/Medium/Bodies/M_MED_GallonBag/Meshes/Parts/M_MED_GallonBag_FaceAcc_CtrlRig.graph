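# Control Rig for the GallonBag character body that retargets finger animation onto claw deform bones and corrects arm deformation.
# When Enable Finger Retarget is on, Map Fingers transfers each joint's local rotation delta from Fingers to Remap Fingers, with per-joint damping, sign-dependent scaling or Z-twist extraction.
# Spherical pose readers on both lower arms then weight corrective upper-arm rotations and elbow offsets, without propagating to children.
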ControlRigBlueprint /BRCosmetics/Characters/Player/Male/Medium/Bodies/M_MED_GallonBag/Meshes/Parts/M_MED_GallonBag_FaceAcc_CtrlRig
graphs [
 New Function_ContainedGraph {nodes [
   n1 = RigVMFunctionEntryNode()
   DISPATCH_RigVMDispatch_ArrayIterator = RigVMDispatchNode(ExecuteContext=(), Array=@n1.Source) {template="DISPATCH_RigVMDispatch_ArrayIterator(in Array,out Element,out Index,out Count,out Ratio)"}
   DISPATCH_RigVMDispatch_CoreEquals = RigVMDispatchNode(B=0, A=@DISPATCH_RigVMDispatch_ArrayIterator.Index) {template="DISPATCH_RigVMDispatch_CoreEquals(in A,in B,out Result)"}
   RigVMFunction_ControlFlowBranch = &RigVMFunction_ControlFlowBranch(ExecuteContext=@DISPATCH_RigVMDispatch_ArrayIterator.ExecuteContext, Condition=@DISPATCH_RigVMDispatch_CoreEquals.Result)
   DISPATCH_RigVMDispatch_ArrayGetAtIndex = RigVMDispatchNode(Index=@DISPATCH_RigVMDispatch_ArrayIterator.Index, Array=@n1.Target) {template="DISPATCH_RigVMDispatch_ArrayGetAtIndex(in Array,in Index,out Element)"}
   n2 = &RigUnit_GetTransform(Space=LocalSpace, bInitial=False, Item=@DISPATCH_RigVMDispatch_ArrayIterator.Element)
   &RigUnit_ModifyTransforms(
    ItemToModify=((Item=(Type=Bone,Name="None"),Transform=(Rotation=(X=0,Y=0,Z=0,W=1),Translation=(X=0,Y=0,Z=0),Scale3D=(X=1,Y=1,Z=1))))
    Weight=1
    WeightMinimum=0
    ItemToModify.0.Item=@DISPATCH_RigVMDispatch_ArrayGetAtIndex.Element
    ItemToModify.0.Transform.Rotation=@n2.Transform.Rotation
    ExecutePin=@RigVMFunction_ControlFlowBranch.False
   )
   n3 = &RigUnit_GetTransform(Space=LocalSpace, bInitial=true, Item=@DISPATCH_RigVMDispatch_ArrayIterator.Element)
   Inverse = /Script/RigVM.RigVMFunction_MathQuaternionInverse(Value=@n3.Transform.Rotation) {is=struct RigVMUnitNode, template="Inverse::Execute(in Value,out Result)"}
   &RigUnit_ModifyTransforms(
    ItemToModify=((Item=(Type=Bone,Name="None"),Transform=(Rotation=(X=0,Y=0,Z=0,W=1),Translation=(X=0,Y=0,Z=0),Scale3D=(X=1,Y=1,Z=1))))
    Weight=1
    WeightMinimum=0
    ItemToModify.0.Item=@DISPATCH_RigVMDispatch_ArrayGetAtIndex.Element
    ItemToModify.0.Transform.Rotation=@Inverse.Result
    ExecutePin=@RigVMFunction_ControlFlowBranch.True
   )
   n4 = RigVMDispatchNode(Index=0, Array=@n1.Source) {template="DISPATCH_RigVMDispatch_ArrayGetAtIndex(in Array,in Index,out Element)"}
   n5 = /Script/RigVM.RigVMFunction_MathQuaternionMul(
    A=&RigUnit_GetTransform(Space=LocalSpace, bInitial=False, Item=@n4.Element).Transform.Rotation
    B=&RigVMFunction_MathQuaternionInverse(Value=&RigUnit_GetTransform(Space=LocalSpace, bInitial=true, Item=@n4.Element).Transform.Rotation) {template="Inverse::Execute(in Value,out Result)"}.Result
   ) {is=struct RigVMUnitNode, template="Multiply::Execute(in A,in B,out Result)"}
   &RigVMFunction_MathQuaternionInverse(Value=@n5.Result) {template="Inverse::Execute(in Value,out Result)"}
   &RigUnit_ModifyTransforms(
    ItemToModify=((Item=(Type=Bone,Name="None"),Transform=(Rotation=(X=0,Y=0,Z=0,W=1),Translation=(X=0,Y=0,Z=0),Scale3D=(X=1,Y=1,Z=1))))
    Weight=0.3
    WeightMinimum=0
    ItemToModify.0.Item=RigVMDispatchNode(Index=0, Array=@n1.Target) {template="DISPATCH_RigVMDispatch_ArrayGetAtIndex(in Array,in Index,out Element)"}.Element
    ItemToModify.0.Transform.Rotation=@n5.Result
    ExecutePin=@RigVMFunction_Sequence.A
   )
   n6 = RigVMDispatchNode(Index=1, Array=@n1.Source) {template="DISPATCH_RigVMDispatch_ArrayGetAtIndex(in Array,in Index,out Element)"}
   n7 = &RigVMFunction_MathQuaternionMul(
    A=&RigUnit_GetTransform(Space=LocalSpace, bInitial=False, Item=@n6.Element).Transform.Rotation
    B=&RigVMFunction_MathQuaternionInverse(Value=&RigUnit_GetTransform(Space=LocalSpace, bInitial=true, Item=@n6.Element).Transform.Rotation) {template="Inverse::Execute(in Value,out Result)"}.Result
   ) {template="Multiply::Execute(in A,in B,out Result)"}
   n8 = RigVMDispatchNode(Index=1, Array=@n1.Target) {template="DISPATCH_RigVMDispatch_ArrayGetAtIndex(in Array,in Index,out Element)"}
   &RigUnit_ModifyTransforms(
    ItemToModify=((Item=(Type=Bone,Name="None"),Transform=(Rotation=(X=0,Y=0,Z=0,W=1),Translation=(X=0,Y=0,Z=0),Scale3D=(X=1,Y=1,Z=1))))
    Weight=1
    WeightMinimum=0
    ExecutePin=@n12.True
    ItemToModify.0.Transform.Rotation=/Script/RigVM.RigVMFunction_MathQuaternionScaleV2(Factor=1.1, Value=@n7.Result) {is=struct RigVMUnitNode, template="Scale::Execute(in Value,in Factor,out Result)"}.Result
    ItemToModify.0.Item=@n8.Element
   )
   Scale = &RigVMFunction_MathQuaternionScaleV2(Factor=0.875, Value=@n7.Result) {template="Scale::Execute(in Value,in Factor,out Result)"}
   &RigUnit_ModifyTransforms(
    ItemToModify=((Item=(Type=Bone,Name="None"),Transform=(Rotation=(X=0,Y=0,Z=0,W=1),Translation=(X=0,Y=0,Z=0),Scale3D=(X=1,Y=1,Z=1))))
    Weight=1
    WeightMinimum=0
    ExecutePin=@n12.False
    ItemToModify.0.Transform.Rotation=@Scale.Result
    ItemToModify.0.Item=@n8.Element
   )
   n9 = RigVMDispatchNode(Index=2, Array=@n1.Source) {template="DISPATCH_RigVMDispatch_ArrayGetAtIndex(in Array,in Index,out Element)"}
   n10 = &RigUnit_GetTransform(Space=LocalSpace, bInitial=False, Item=@n9.Element)
   n11 = &RigVMFunction_MathQuaternionInverse(Value=&RigUnit_GetTransform(Space=LocalSpace, bInitial=true, Item=@n9.Element).Transform.Rotation) {template="Inverse::Execute(in Value,out Result)"}
   &RigUnit_ModifyTransforms(
    ItemToModify=((Item=(Type=Bone,Name="None"),Transform=(Rotation=(X=0,Y=0,Z=0,W=1),Translation=(X=0,Y=0,Z=0),Scale3D=(X=1,Y=1,Z=1))))
    Weight=0.5
    WeightMinimum=0
    ItemToModify.0.Item=RigVMDispatchNode(Index=2, Array=@n1.Target) {template="DISPATCH_RigVMDispatch_ArrayGetAtIndex(in Array,in Index,out Element)"}.Element
    ItemToModify.0.Transform.Rotation=&RigVMFunction_MathQuaternionSwingTwist(
     TwistAxis=(X=0,Y=0,Z=1)
     Input=&RigVMFunction_MathQuaternionMul(A=@n10.Transform.Rotation, B=@n11.Result) {template="Multiply::Execute(in A,in B,out Result)"}.Result
    ).Twist
    ExecutePin=@RigVMFunction_Sequence.C
   )
   Multiply = &RigVMFunction_MathQuaternionMul(A=@n2.Transform.Rotation, B=@Inverse.Result) {template="Multiply::Execute(in A,in B,out Result)"}
   &RigVMFunction_MathQuaternionMul(A=@n2.Transform.Rotation, B=@n3.Transform.Rotation) {template="Multiply::Execute(in A,in B,out Result)"}
   RigVMFunctionReturnNode(ExecuteContext=@RigVMFunction_Sequence.D)
   &RigVMFunction_ControlFlowBranch(
    ExecuteContext=()
    Condition=RigVMDispatchNode(B=0, A=@DISPATCH_RigVMDispatch_ArrayIterator.Index) {template="DISPATCH_RigVMDispatch_CoreEquals(in A,in B,out Result)"}.Result
   )
   RigVMFunction_MathQuaternionToEuler = &RigVMFunction_MathQuaternionToEuler(RotationOrder=ZYX, Value=@n7.Result)
   Greater = /Script/RigVM.RigVMFunction_MathDoubleGreater(B=0, A=@RigVMFunction_MathQuaternionToEuler.Result.Z) {is=struct RigVMUnitNode, template="Greater::Execute(in A,in B,out Result)"}
   n12 = &RigVMFunction_ControlFlowBranch(Condition=@Greater.Result, ExecuteContext=@RigVMFunction_Sequence.B)
   RigVMFunction_Sequence = RigVMAggregateNode(ExecuteContext=@n1.ExecuteContext)
  ]}
 RigVMFunctionLibrary {nodes [Map Fingers = RigVMCollapseNode(ExecuteContext=(), Inverse=False, Source=(), Target=())]}
 RigVMFunction_Sequence_1_ContainedGraph {nodes [
   n1 = RigVMFunctionEntryNode()
   RigVMFunction_Sequence = &RigVMFunction_Sequence(ExecuteContext=@n1.ExecuteContext)
   n2 = &RigVMFunction_Sequence(ExecuteContext=@RigVMFunction_Sequence.B)
   RigVMFunctionReturnNode(A=@RigVMFunction_Sequence.A, B=@n2.A, C=@n2.B)
  ]}
 RigVMFunction_Sequence_1_ContainedGraph {nodes [
   n1 = RigVMFunctionEntryNode()
   RigVMFunction_Sequence = &RigVMFunction_Sequence(ExecuteContext=@n1.ExecuteContext)
   n2 = &RigVMFunction_Sequence(ExecuteContext=@RigVMFunction_Sequence.B)
   n3 = &RigVMFunction_Sequence(ExecuteContext=@n2.B)
   RigVMFunctionReturnNode(A=@RigVMFunction_Sequence.A, B=@n2.A, C=@n3.A, D=@n3.B)
  ]}
 RigVMModel {nodes [
   RigUnit_BeginExecution = &RigUnit_BeginExecution()
   RigVMFunction_Sequence = RigVMAggregateNode(ExecuteContext=@RigUnit_BeginExecution.ExecutePin)
   n1 = &RigUnit_SphericalPoseReader(
    ActiveRegionScaleFactors=(PositiveWidth=5,NegativeWidth=1,PositiveHeight=3,NegativeHeight=3)
    ActiveRegionSize=0.1
    Debug=(bDrawDebug=True,bDraw2D=False,bDrawLocalAxes=False,DebugScale=25,DebugSegments=20,DebugThickness=0.25)
    DriverAxis=(X=-1,Y=0,Z=0)
    DriverItem=(Type=Bone,Name="def_lowerarm_r")
    FalloffRegionScaleFactors=(PositiveWidth=1,NegativeWidth=1,PositiveHeight=1,NegativeHeight=1)
    FalloffSize=0.3
    FlipHeightScaling=False
    FlipWidthScaling=False
    OptionalParentItem=(Type=Bone,Name="def_upperarm_r")
    RotationOffset=(X=-90,Y=0,Z=0)
    ExecutePin=@RigVMFunction_Sequence.B
   )
   > RigUnit_SetRelativeTransformForItem = /Script/ControlRig.RigUnit_SetRelativeTransformForItem(
    Child=(Type=Bone,Name="def_upperarm_r")
    Parent=(Type=Bone,Name="def_upperarm_r")
    Value=(Rotation=(X=0,Y=0,Z=-0.0872,W=0.9962),Translation=(X=0,Y=0,Z=0),Scale3D=(X=1,Y=1,Z=1))
    bParentInitial=False
    bPropagateToChildren=false
    Weight=@n1.OutputParam
   ) {
    is=struct RigVMUnitNode
    template="Set Relative Transform::Execute(in Child,in Parent,in bParentInitial,in Value,in Weight,in bPropagateToChildren)"
   }
   > &RigUnit_SetRelativeTransformForItem(
    Child=(Type=Bone,Name="def_upperarm_elbow_r")
    Parent=(Type=Bone,Name="def_upperarm_elbow_r")
    Value=(Rotation=(X=0,Y=0,Z=0,W=1),Translation=(X=1.5,Y=2,Z=0),Scale3D=(X=1,Y=1,Z=1))
    bParentInitial=False
    bPropagateToChildren=false
    Weight=@n1.OutputParam
   ) {template="Set Relative Transform::Execute(in Child,in Parent,in bParentInitial,in Value,in Weight,in bPropagateToChildren)"}
   n2 = &RigUnit_SphericalPoseReader(
    ActiveRegionScaleFactors=(PositiveWidth=5,NegativeWidth=1,PositiveHeight=3,NegativeHeight=3)
    ActiveRegionSize=0.1
    Debug=(bDrawDebug=True,bDraw2D=False,bDrawLocalAxes=False,DebugScale=25,DebugSegments=20,DebugThickness=0.25)
    DriverAxis=(X=1,Y=0,Z=0)
    DriverItem=(Type=Bone,Name="def_lowerarm_l")
    FalloffRegionScaleFactors=(PositiveWidth=1,NegativeWidth=1,PositiveHeight=1,NegativeHeight=1)
    FalloffSize=0.3
    FlipHeightScaling=true
    FlipWidthScaling=true
    OptionalParentItem=(Type=Bone,Name="def_upperarm_l")
    RotationOffset=(X=90,Y=0,Z=0)
    ExecutePin=@RigVMFunction_Sequence.C
   )
   > &RigUnit_SetRelativeTransformForItem(
    Child=(Type=Bone,Name="def_upperarm_l")
    Parent=(Type=Bone,Name="def_upperarm_l")
    Value=(Rotation=(X=0,Y=0,Z=-0.0872,W=0.9962),Translation=(X=0,Y=0,Z=0),Scale3D=(X=1,Y=1,Z=1))
    bParentInitial=False
    bPropagateToChildren=false
    Weight=@n2.OutputParam
   ) {template="Set Relative Transform::Execute(in Child,in Parent,in bParentInitial,in Value,in Weight,in bPropagateToChildren)"}
   > &RigUnit_SetRelativeTransformForItem(
    Child=(Type=Bone,Name="def_upperarm_elbow_l")
    Parent=(Type=Bone,Name="def_upperarm_elbow_l")
    Value=(Rotation=(X=0,Y=0,Z=0,W=1),Translation=(X=-1.5,Y=-2,Z=0),Scale3D=(X=1,Y=1,Z=1))
    bParentInitial=False
    bPropagateToChildren=false
    Weight=@n2.OutputParam
   ) {template="Set Relative Transform::Execute(in Child,in Parent,in bParentInitial,in Value,in Weight,in bPropagateToChildren)"}
   RigVMFunction_ControlFlowBranch = &RigVMFunction_ControlFlowBranch(Condition=RigVMVariableNode() {variable=Enable Finger Retarget}.Value, ExecuteContext=@RigVMFunction_Sequence.A)
   n3 = RigVMDispatchNode(ExecuteContext=@RigVMFunction_ControlFlowBranch.True, Array=RigVMVariableNode() {variable=Fingers}.Value) {template="DISPATCH_RigVMDispatch_ArrayIterator(in Array,out Element,out Index,out Count,out Ratio)"}
   > Map Fingers = RigVMFunctionReferenceNode(
    Inverse=False
    Target=RigVMDispatchNode(Index=@n3.Index, Array=RigVMVariableNode() {variable=Remap Fingers}.Value) {template="DISPATCH_RigVMDispatch_ArrayGetAtIndex(in Array,in Index,out Element)"}.Element.Keys
    Source=@n3.Element.Keys
   )
  ]}
]
hierarchy [
 root {loc=<0 0 0>, rot=<0 0 0 1>}
 pelvis {parent=root, loc=<0 -0.7071 93.1191>, rot=<0.0127 0.707 -0.0127 -0.707>}
 spine_01 {parent=pelvis, loc=<6.7494 -0.6388 0>, rot=<0 0 0.034 -0.9994>}
 spine_02 {parent=spine_01, loc=<7.2518 -0.035 0>, rot=<0 0 0.0238 -0.9997>}
 spine_03 {parent=spine_02, loc=<6.0104 0 0>, rot=<0 0 0.0149 -0.9999>}
 spine_04 {parent=spine_03, loc=<8.1024 0 0>, rot=<0 0 -0.0132 -0.9999>}
 spine_05 {parent=spine_04, loc=<10.0146 0 0>, rot=<0 0 -0.0015 -1>}
 neck_01 {parent=spine_05, loc=<11.3138 0.1529 0>, rot=<0 0 -0.2003 -0.9797>}
 neck_02 {parent=neck_01, loc=<4.5627 0 0>, rot=<0 0 0.0504 -0.9987>}
 head {parent=neck_02, loc=<5.3914 0 0>, rot=<0 0 0.115 -0.9934>}
 clavicle_l {parent=spine_05, loc=<4.3575 -0.1466 -2.3223>, rot=<0.0465 -0.7289 -0.009 -0.6829>}
 upperarm_l {parent=clavicle_l, loc=<15.5111 0 0>, rot=<0.0226 -0.3234 -0.0088 -0.9459>}
 lowerarm_l {parent=upperarm_l, loc=<26.6514 0 0>, rot=<0.0391 -0.0099 -0.2453 -0.9686>}
 hand_l {parent=lowerarm_l, loc=<22.3491 0 0>, rot=<-0.7081 -0.0507 0.0975 -0.6975>}
 thumb_01_l {parent=hand_l, loc=<1.3118 -0.3191 -2.459>, rot=<0.6546 -0.4042 0.0664 -0.6354>}
 thumb_02_l {parent=thumb_01_l, loc=<5.9336 -0.0448 0.0401>, rot=<-0.0037 0.0213 0.298 -0.9543>}
 thumb_03_l {parent=thumb_02_l, loc=<2.2223 0.3682 -0.0907>, rot=<0.0118 -0.013 0.0066 -0.9998>}
 index_metacarpal_l {parent=hand_l, loc=<1.6535 0.0183 -2.96>, rot=<0.0158 -0.0124 -0.0173 -0.9996>}
 index_01_l {parent=index_metacarpal_l, loc=<8.3008 0.0357 -0.0524>, rot=<-0.0044 0.0485 0.0924 -0.9945>}
 index_02_l {parent=index_01_l, loc=<3.3587 -0.3004 0.0452>, rot=<-0.002 0.0027 0.3225 -0.9466>}
 index_03_l {parent=index_02_l, loc=<2.5578 -0.123 -0.0252>, rot=<-0.002 0.0067 0.2558 -0.9667>}
 middle_metacarpal_l {parent=hand_l, loc=<2.1856 -0.1725 -0.5984>, rot=<0.006 0.0168 -0.0329 -0.9993>}
 middle_01_l {parent=middle_metacarpal_l, loc=<7.3275 -0.2447 0.0129>, rot=<-0.0208 0.0042 0.0858 -0.9961>}
 middle_02_l {parent=middle_01_l, loc=<4.0123 0.1401 0.1175>, rot=<-0.0218 0.0258 0.3512 -0.9357>}
 middle_03_l {parent=middle_02_l, loc=<2.6077 0.0534 0.1373>, rot=<0.0082 0.0188 0.1708 -0.9851>}
 ring_metacarpal_l {parent=hand_l, loc=<2.2074 -0.3625 1.3638>, rot=<-0.0284 0.0722 -0.0458 -0.9959>}
 ring_01_l {parent=ring_metacarpal_l, loc=<7.0295 -0.6197 -0.1446>, rot=<-0.0489 -0.0252 0.1892 -0.9804>}
 ring_02_l {parent=ring_01_l, loc=<3.6174 0.4647 -0.1027>, rot=<0 0 0.2806 -0.9598>}
 ring_03_l {parent=ring_02_l, loc=<2.4152 0.114 0.0199>, rot=<0 0 0.1189 -0.9929>}
 pinky_metacarpal_l {parent=hand_l, loc=<1.7616 -0.9753 3.3814>, rot=<-0.0397 0.0814 -0.069 -0.9935>}
 pinky_01_l {parent=pinky_metacarpal_l, loc=<7.1709 -0.4831 -0.1094>, rot=<-0.0654 -0.0035 0.2242 -0.9723>}
 pinky_02_l {parent=pinky_01_l, loc=<2.8274 0.0893 -0.1043>, rot=<0.0037 0.0011 0.2854 -0.9584>}
 pinky_03_l {parent=pinky_02_l, loc=<2.2046 0.0307 -0.0078>, rot=<-0.0002 0.0098 0.1445 -0.9895>}
 weapon_l {parent=hand_l, loc=<-0.1974 -3.4111 0.3811>, rot=<0 0 0 1>}
 lowerarm_twist_02_l {parent=lowerarm_l, loc=<9.6199 0 0>, rot=<0 0 0 -1>}
 lowerarm_twist_01_l {parent=lowerarm_l, loc=<17.8549 0 0>, rot=<0 0 0 -1>}
 upperarm_twist_01_l {parent=upperarm_l, loc=<0.0041 0 0>, rot=<0 0 0 1>}
 upperarm_twist_02_l {parent=upperarm_l, loc=<13.7941 0 0>, rot=<0 0 0 1>}
 clavicle_r {parent=spine_05, loc=<4.3575 -0.1466 2.3223>, rot=<0.7289 0.0465 -0.6829 0.009>}
 upperarm_r {parent=clavicle_r, loc=<-15.5111 0 0>, rot=<0.0226 -0.3234 -0.0088 -0.9459>}
 lowerarm_r {parent=upperarm_r, loc=<-26.6514 0 0>, rot=<0.0391 -0.0099 -0.2453 -0.9686>}
 hand_r {parent=lowerarm_r, loc=<-22.3491 0 0>, rot=<-0.7081 -0.0507 0.0975 -0.6975>}
 weapon_r {parent=hand_r, loc=<0.1974 3.4111 -0.3811>, rot=<0 0 0 1>}
 thumb_01_r {parent=hand_r, loc=<-1.3118 0.3191 2.459>, rot=<0.6546 -0.4042 0.0664 -0.6354>}
 thumb_02_r {parent=thumb_01_r, loc=<-5.9336 0.0448 -0.0401>, rot=<-0.0037 0.0213 0.298 -0.9543>}
 thumb_03_r {parent=thumb_02_r, loc=<-2.2223 -0.3682 0.0907>, rot=<0.0118 -0.013 0.0066 -0.9998>}
 index_metacarpal_r {parent=hand_r, loc=<-1.6535 -0.0183 2.96>, rot=<0.0158 -0.0124 -0.0173 -0.9996>}
 index_01_r {parent=index_metacarpal_r, loc=<-8.3008 -0.0357 0.0524>, rot=<-0.0044 0.0485 0.0924 -0.9945>}
 index_02_r {parent=index_01_r, loc=<-3.3587 0.3004 -0.0452>, rot=<-0.002 0.0027 0.3225 -0.9466>}
 index_03_r {parent=index_02_r, loc=<-2.5578 0.123 0.0252>, rot=<-0.002 0.0067 0.2558 -0.9667>}
 middle_metacarpal_r {parent=hand_r, loc=<-2.1856 0.1725 0.5984>, rot=<0.006 0.0168 -0.0329 -0.9993>}
 middle_01_r {parent=middle_metacarpal_r, loc=<-7.3275 0.2447 -0.0129>, rot=<-0.0208 0.0042 0.0858 -0.9961>}
 middle_02_r {parent=middle_01_r, loc=<-4.0123 -0.1401 -0.1175>, rot=<-0.0218 0.0258 0.3512 -0.9357>}
 middle_03_r {parent=middle_02_r, loc=<-2.6077 -0.0534 -0.1373>, rot=<0.0082 0.0188 0.1708 -0.9851>}
 ring_metacarpal_r {parent=hand_r, loc=<-2.2074 0.3625 -1.3638>, rot=<-0.0284 0.0722 -0.0458 -0.9959>}
 ring_01_r {parent=ring_metacarpal_r, loc=<-7.0295 0.6197 0.1446>, rot=<-0.0489 -0.0252 0.1892 -0.9804>}
 ring_02_r {parent=ring_01_r, loc=<-3.6174 -0.4647 0.1027>, rot=<0 0 0.2806 -0.9598>}
 ring_03_r {parent=ring_02_r, loc=<-2.4152 -0.114 -0.0199>, rot=<0 0 0.1189 -0.9929>}
 pinky_metacarpal_r {parent=hand_r, loc=<-1.7616 0.9753 -3.3814>, rot=<-0.0397 0.0814 -0.069 -0.9935>}
 pinky_01_r {parent=pinky_metacarpal_r, loc=<-7.1709 0.4831 0.1094>, rot=<-0.0654 -0.0035 0.2242 -0.9723>}
 pinky_02_r {parent=pinky_01_r, loc=<-2.8274 -0.0893 0.1043>, rot=<0.0037 0.0011 0.2854 -0.9584>}
 pinky_03_r {parent=pinky_02_r, loc=<-2.2046 -0.0307 0.0078>, rot=<-0.0002 0.0098 0.1445 -0.9895>}
 lowerarm_twist_02_r {parent=lowerarm_r, loc=<-9.6199 0 0>, rot=<0 0 0 -1>}
 lowerarm_twist_01_r {parent=lowerarm_r, loc=<-17.8549 0 0>, rot=<0 0 0 -1>}
 upperarm_twist_01_r {parent=upperarm_r, loc=<-0.0041 0 0>, rot=<0 0 0 -1>}
 upperarm_twist_02_r {parent=upperarm_r, loc=<-13.7941 0 0>, rot=<0 0 0 -1>}
 deform_spine_01 {parent=pelvis, loc=<1.5165 0 0>, rot=<0 0 0.034 -0.9994>}
 deform_spine_02 {parent=deform_spine_01, loc=<8.3267 -0.1492 0>, rot=<0 0 0.0238 -0.9997>}
 deform_spine_03 {parent=deform_spine_02, loc=<6.038 0.0264 0>, rot=<0 0 0.0149 -0.9999>}
 deform_spine_04 {parent=deform_spine_03, loc=<7.6524 0 0>, rot=<0 0 -0.0132 -0.9999>}
 deform_spine_05 {parent=deform_spine_04, loc=<14.6293 0.0082 0>, rot=<0 0 -0.0015 -1>}
 deform_clavicle_l {parent=deform_spine_05, loc=<4.3575 -0.1466 -2.3223>, rot=<0.0465 -0.7289 -0.009 -0.6829>}
 def_upperarm_l {parent=deform_clavicle_l, loc=<17.9459 0.1205 -1.8834>, rot=<0.0089 -0.324 0.0023 -0.946>}
 def_upperarm_elbow_l {parent=def_upperarm_l, loc=<19.8165 0 0>, rot=<0.0372 -0.0081 -0.0971 -0.9946>}
 def_lowerarm_l {parent=def_upperarm_elbow_l, loc=<5.106 0 0>, rot=<0.0603 0.0063 -0.1531 -0.9864>}
 def_hand_l {parent=def_lowerarm_l, loc=<20.855 0 0>, rot=<-0.7359 -0.0573 0.0943 -0.668>}
 deform_claw_a_l_01 {parent=def_hand_l, loc=<6.4283 -1.316 -1.2371>, rot=<-0.1423 0.058 -0.0823 -0.9847>}
 deform_claw_a_l_02 {parent=deform_claw_a_l_01, loc=<6.1207 0 0.0061>, rot=<-0.024 -0.0199 0.1212 -0.9921>}
 deform_claw_a_l_03 {parent=deform_claw_a_l_02, loc=<4.7207 0 0.1596>, rot=<0 0 0.3207 -0.9472>}
 deform_claw_b_l_01 {parent=def_hand_l, loc=<6.175 -2.0671 0.7315>, rot=<-0.1423 0.058 -0.0823 -0.9847>}
 deform_claw_b_l_02 {parent=deform_claw_b_l_01, loc=<6.1207 0 0.0062>, rot=<-0.0242 -0.0197 0.1117 -0.9933>}
 deform_claw_b_l_03 {parent=deform_claw_b_l_02, loc=<4.7922 0 0.1104>, rot=<0 0 0.3292 -0.9443>}
 deform_claw_c_l_01 {parent=def_hand_l, loc=<5.224 -3.9667 0.6938>, rot=<0.8223 -0.5359 -0.0741 -0.1768>}
 deform_claw_c_l_02 {parent=deform_claw_c_l_01, loc=<6.1207 0 0.0062>, rot=<-0.0242 -0.0197 0.1102 -0.9934>}
 deform_claw_c_l_03 {parent=deform_claw_c_l_02, loc=<4.8054 0 0.0631>, rot=<0 0 0.338 -0.9411>}
 deform_claw_d_l_01 {parent=def_hand_l, loc=<5.2656 -2.8159 -2.6259>, rot=<0.8361 -0.5145 -0.0677 -0.1779>}
 deform_claw_d_l_02 {parent=deform_claw_d_l_01, loc=<6.1207 0 0.0062>, rot=<-0.0247 -0.019 0.0839 -0.996>}
 deform_claw_d_l_03 {parent=deform_claw_d_l_02, loc=<4.7543 0 0.1774>, rot=<0 0 0.3304 -0.9438>}
 deform_fake_claw_c_l_01 {parent=def_hand_l, loc=<5.9774 -2.6969 2.8028>, rot=<-0.1423 0.058 -0.0823 -0.9847>}
 deform_fake_claw_c_l_02 {parent=deform_fake_claw_c_l_01, loc=<6.1207 0 0.0061>, rot=<-0.0242 -0.0197 0.1117 -0.9933>}
 deform_fake_claw_c_l_03 {parent=deform_fake_claw_c_l_02, loc=<4.7922 0 0.1104>, rot=<0 0 0.3292 -0.9443>}
 deform_fake_claw_d_l_01 {parent=def_hand_l, loc=<5.7793 -3.3281 4.8787>, rot=<-0.1423 0.058 -0.0823 -0.9847>}
 deform_fake_claw_d_l_02 {parent=deform_fake_claw_d_l_01, loc=<6.1207 0 0.0061>, rot=<-0.0242 -0.0197 0.1117 -0.9933>}
 deform_fake_claw_d_l_03 {parent=deform_fake_claw_d_l_02, loc=<4.7922 0 0.1104>, rot=<0 0 0.3292 -0.9443>}
 deform_clavicle_r {parent=deform_spine_05, loc=<4.3578 -0.1466 2.3223>, rot=<0.7289 0.0465 -0.6829 0.009>}
 def_upperarm_r {parent=deform_clavicle_r, loc=<-17.946 -0.1205 1.8833>, rot=<0.0089 -0.324 0.0023 -0.946>}
 def_upperarm_elbow_r {parent=def_upperarm_r, loc=<-19.8167 0 0.0003>, rot=<0.0372 -0.0081 -0.0971 -0.9946>}
 def_lowerarm_r {parent=def_upperarm_elbow_r, loc=<-5.1059 0 -0.0001>, rot=<0.0603 0.0063 -0.1531 -0.9864>}
 def_hand_r {parent=def_lowerarm_r, loc=<-20.8549 0 -0.0001>, rot=<-0.7359 -0.0573 0.0943 -0.668>}
 deform_claw_a_r_01 {parent=def_hand_r, loc=<-6.4286 1.3163 1.237>, rot=<-0.1423 0.058 -0.0823 -0.9847>}
 deform_claw_a_r_02 {parent=deform_claw_a_r_01, loc=<-6.1208 0 -0.0061>, rot=<-0.024 -0.0199 0.1212 -0.9921>}
 deform_claw_a_r_03 {parent=deform_claw_a_r_02, loc=<-4.7206 0 -0.1596>, rot=<0 0 0.3207 -0.9472>}
 deform_claw_b_r_01 {parent=def_hand_r, loc=<-6.1754 2.0673 -0.7316>, rot=<-0.1423 0.058 -0.0823 -0.9847>}
 deform_claw_b_r_02 {parent=deform_claw_b_r_01, loc=<-6.1207 0 -0.0061>, rot=<-0.0242 -0.0197 0.1117 -0.9933>}
 deform_claw_b_r_03 {parent=deform_claw_b_r_02, loc=<-4.7922 0 -0.1104>, rot=<0 0 0.3292 -0.9443>}
 deform_claw_c_r_01 {parent=def_hand_r, loc=<-5.2243 3.967 -0.6939>, rot=<0.8223 -0.5359 -0.0741 -0.1768>}
 deform_claw_c_r_02 {parent=deform_claw_c_r_01, loc=<-6.1207 0 -0.0062>, rot=<-0.0242 -0.0197 0.1102 -0.9934>}
 deform_claw_c_r_03 {parent=deform_claw_c_r_02, loc=<-4.8054 0 -0.0631>, rot=<0 0 0.338 -0.9411>}
 deform_claw_d_r_01 {parent=def_hand_r, loc=<-5.266 2.8161 2.6257>, rot=<0.8361 -0.5145 -0.0677 -0.1779>}
 deform_claw_d_r_02 {parent=deform_claw_d_r_01, loc=<-6.1208 -0.0001 -0.0062>, rot=<-0.0247 -0.019 0.0839 -0.996>}
 deform_claw_d_r_03 {parent=deform_claw_d_r_02, loc=<-4.7542 0 -0.1774>, rot=<0 0 0.3304 -0.9438>}
 deform_fake_claw_c_r_01 {parent=def_hand_r, loc=<-5.9778 2.6971 -2.8029>, rot=<-0.1423 0.058 -0.0823 -0.9847>}
 deform_fake_claw_c_r_02 {parent=deform_fake_claw_c_r_01, loc=<-6.1207 0 -0.0061>, rot=<-0.0242 -0.0197 0.1117 -0.9933>}
 deform_fake_claw_c_r_03 {parent=deform_fake_claw_c_r_02, loc=<-4.7922 0 -0.1104>, rot=<0 0 0.3292 -0.9443>}
 deform_fake_claw_d_r_01 {parent=def_hand_r, loc=<-5.7797 3.3283 -4.8788>, rot=<-0.1423 0.058 -0.0823 -0.9847>}
 deform_fake_claw_d_r_02 {parent=deform_fake_claw_d_r_01, loc=<-6.1207 0 -0.0061>, rot=<-0.0242 -0.0197 0.1117 -0.9933>}
 deform_fake_claw_d_r_03 {parent=deform_fake_claw_d_r_02, loc=<-4.7922 0 -0.1104>, rot=<0 0 0.3292 -0.9443>}
]
variables [
 Enable Finger Retarget {type=bool, default=True}
 Fingers {
  type="TArray<FRigElementKeyCollection>"
  default="((Keys=((Type=Bone,Name=\"index_01_l\"),(Type=Bone,Name=\"index_02_l\"),(Type=Bone,Name=\"index_03_l\"))),(Keys=((Type=Bone,Name=\"middle_01_l\"),(Type=Bone,Name=\"middle_02_l\"),(Type=Bone,Name=\"middle_03_l\"))),(Keys=((Type=Bone,Name=\"ring_01_l\"),(Type=Bone,Name=\"r ...(truncated)"
 }
 Get Delta Ref_Result {type=FEulerTransform, default=(Location=(X=0,Y=0,Z=0),Rotation=(Pitch=0,Yaw=0,Roll=0),Scale=(X=1,Y=1,Z=1))}
 Remap Fingers {
  type="TArray<FRigElementKeyCollection>"
  default="((Keys=((Type=Bone,Name=\"deform_claw_a_l_01\"),(Type=Bone,Name=\"deform_claw_a_l_02\"),(Type=Bone,Name=\"deform_claw_a_l_03\"))),(Keys=((Type=Bone,Name=\"deform_claw_b_l_01\"),(Type=Bone,Name=\"deform_claw_b_l_02\"),(Type=Bone,Name=\"deform_claw_b_l_03\"))),(Keys=((T ...(truncated)"
 }
]
