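# Anim Blueprint for the GallonBag character's face accessory mesh, which retargets the body pose onto its own skeleton.
# IK-retargeted pose is layered with a copied mesh pose from the clavicles down, then passed through a Control Rig.
# DefaultSlot overrides claw bones, weighted by the front end's Frontend_Idle_Bypass curve.
# That curve at 1 disables finger retargeting and selects the No Fingers profile, which weapon changes overwrite with the current or default profile.
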AnimBlueprint /BRCosmetics/Characters/Player/Male/Medium/Bodies/M_MED_GallonBag/Meshes/Parts/M_MED_GallonBag_FaceAcc_AnimBP
target_skeleton=&/M_MED_GallonBag_FaceAcc_Skeleton
graphs [
 AnimGraph {nodes [
   n1 = Get Output() -> (Output:Skeletal Mesh Component Budgeted Object Reference)
   n2 = Get Retarget Profile() -> (Retarget Profile:Retarget Profile Structure)
   n3 = Control Rig!(
    Source=Layered blend per bone(
     BasePose=Retarget Pose From Mesh!(CustomRetargetProfile:Retarget Profile Structure=@n2) -> (Pose:Pose Link Structure) {
      is=title AnimGraphNode_RetargetPoseFromMesh
      iKRetargeterAsset=&/M_MED_GallonBag_FaceAcc_Retargeter
     }
     BlendPoses_0=Copy Pose From Mesh(SourceMeshComponent=@n1) {bUseAttachedParent=true}
    ) {layerSetup [{branchFilters [{boneName=clavicle_l}, {boneName=clavicle_r}]}]}
    Alpha
    Enable Finger Retarget:bool=true
   ) -> (Pose) {
    bTransferInputCurves=false
    alphaBoolBlend {blendOption=Linear}
    controlRigAssetReference {blueprintRigClass=&/M_MED_GallonBag_FaceAcc_CtrlRig.M_MED_GallonBag_FaceAcc_CtrlRig_C}
   }
   n4 = Layered blend per bone(BasePose=@n3, BlendPoses_0=Slot 'DefaultSlot'(), BlendWeights_0=Get CustomIdleIsPlayingFloat() -> (CustomIdleIsPlayingFloat:float)) {layerSetup [{branchFilters [
       {boneName=deform_claw_a_l_01}
       {boneName=deform_claw_a_l_02}
       {boneName=deform_claw_a_l_03}
       {boneName=deform_claw_b_l_01}
       {boneName=deform_claw_b_l_02}
       {boneName=deform_claw_b_l_03}
       {boneName=deform_claw_c_l_01}
       {boneName=deform_claw_c_l_02}
       {boneName=deform_claw_c_l_03}
       {boneName=deform_claw_d_l_01}
       {boneName=deform_claw_d_l_02}
       {boneName=deform_claw_d_l_03}
       {boneName=deform_claw_a_r_01}
       {boneName=deform_claw_a_r_02}
       {boneName=deform_claw_a_r_03}
       {boneName=deform_claw_b_r_01}
       {boneName=deform_claw_b_r_02}
       {boneName=deform_claw_b_r_03}
       {boneName=deform_claw_c_r_01}
       {boneName=deform_claw_c_r_02}
       {boneName=deform_claw_c_r_03}
       {boneName=deform_claw_d_r_01}
       {boneName=deform_claw_d_r_02}
       {boneName=deform_claw_d_r_03}
      ]}]}
   Output Pose(Result=@n4) {layerGroup=None}
   Get Enable Finger Retarget() -> (Enable Finger Retarget:bool)
  ]}
 EventGraph {nodes [
   Event BlueprintUpdateAnimation()
   > n1 = Branch(Condition=Get bIsInFrontEnd())
   > n2 = Set CustomIdleIsPlayingFloat(CustomIdleIsPlayingFloat:float=GetCurveValue(self=GetOwnerFrontEndAnimInstance(), CurveName=Frontend_Idle_Bypass)) -> (Output_Get:float)
   > n3 = Branch(Condition="Equal (Float)"(A=@n2, B=1))
   > Set Enable Finger Retarget(Enable Finger Retarget:bool=false) -> (Output_Get:bool)
   > Set Retarget Profile(Retarget Profile:Retarget Profile Structure=Get No Fingers Retarget Profile() -> (No Fingers Retarget Profile:Retarget Profile Structure)) -> (Output_Get:Retarget Profile Structure)
   Set Enable Finger Retarget(execute=@n3.else, Enable Finger Retarget=true)
   > Set Retarget Profile(Retarget Profile=Get Default Retarget Profile() -> (Default Retarget Profile:Retarget Profile Structure))
   Set Enable Finger Retarget(execute=@n1.else, Enable Finger Retarget=true)
   > Set Retarget Profile(Retarget Profile=Get Default Retarget Profile())
   Event OnWeaponChanged()
   > n4 = Cast To FortWeaponRanged!(execute, Object=Get CurrentWeapon() -> (CurrentWeapon:Fort Weapon Object Reference)) -> (then, CastFailed, AsFort Weapon Ranged:Fort Weapon Ranged Object Reference) {is=title K2Node_DynamicCast}
   > Set No Fingers Retarget Profile(No Fingers Retarget Profile:Retarget Profile Structure=Get Retarget Profile()) -> (Output_Get:Retarget Profile Structure)
   Set No Fingers Retarget Profile(execute=@n4.CastFailed, No Fingers Retarget Profile=Get Default Retarget Profile())
   n5 = Event OnCharacterCustomizationLoaded()
   > Set Output(Output:Skeletal Mesh Component Budgeted Object Reference=Get Array Item with EFort Custom Part Type(Array=@n5.CustomMeshParts)) -> (Output_Get:Skeletal Mesh Component Budgeted Object Reference)
  ]}
]
variables [
 CustomIdleIsPlayingFloat {type=number}
 Default Retarget Profile {
  type=RetargetProfile
  default {
   bApplyChainSettings=false
   bApplyGlobalSettings=false
   bApplyRootSettings=false
   bApplySourceRetargetPose=false
   bApplyTargetRetargetPose=false
   bForceAllIKOff=false
   sourceRetargetPoseName=None
   targetRetargetPoseName=None
   chainSettings {}
   globalSettings {
    bCopyBasePose=false
    bEnableFK=true
    bEnableIK=true
    bEnablePost=true
    bEnableRoot=true
    bWarping=false
    copyBasePoseRoot=None
    directionChain=None
    directionSource=Goals
    forwardDirection=Y
    sidewaysOffset=0
    sourceScaleFactor=1
    warpForwards=1
    warpSplay=1
   }
   retargetOpProfiles []
   rootSettings {
    affectIKHorizontal=1
    affectIKVertical=0
    blendToSource=0
    rotationAlpha=1
    scaleHorizontal=1
    scaleVertical=1
    translationAlpha=1
    blendToSourceWeights=<1 1 1>
    rotationOffset=<rot 0 0 0>
    translationOffset=<0 0 0>
   }
  }
 }
 Enable Finger Retarget {type=boolean, default=true}
 No Fingers Retarget Profile {
  type=RetargetProfile
  default {
   bApplyChainSettings=true
   bApplyGlobalSettings=false
   bApplyRootSettings=false
   bApplySourceRetargetPose=false
   bApplyTargetRetargetPose=false
   bForceAllIKOff=false
   sourceRetargetPoseName=None
   targetRetargetPoseName=None
   chainSettings {
    leftIndex {
     fK {
      enableFK=false
      poleVectorMaintainOffset=false
      poleVectorMatching=0
      poleVectorOffset=0
      rotationAlpha=1
      rotationMode=Interpolated
      translationAlpha=1
      translationMode=None
     }
     iK {
      bAffectedByIKWarping=true
      blendToSource=1
      blendToSourceRotation=0
      blendToSourceTranslation=1
      enableIK=false
      extension=1
      scaleVertical=1
      blendToSourceWeights=<1 1 1>
      staticLocalOffset=<0 0 0>
      staticOffset=<0 0 0>
      staticRotationOffset=<rot 0 0 0>
     }
     speedPlanting {enableSpeedPlanting=false, speedCurveName=None, speedThreshold=15, unplantCriticalDamping=1, unplantStiffness=250}
    }
    leftMiddle {
     fK {
      enableFK=false
      poleVectorMaintainOffset=false
      poleVectorMatching=0
      poleVectorOffset=0
      rotationAlpha=1
      rotationMode=Interpolated
      translationAlpha=1
      translationMode=None
     }
     iK {
      bAffectedByIKWarping=true
      blendToSource=1
      blendToSourceRotation=0
      blendToSourceTranslation=1
      enableIK=false
      extension=1
      scaleVertical=1
      blendToSourceWeights=<1 1 1>
      staticLocalOffset=<0 0 0>
      staticOffset=<0 0 0>
      staticRotationOffset=<rot 0 0 0>
     }
     speedPlanting {enableSpeedPlanting=false, speedCurveName=None, speedThreshold=15, unplantCriticalDamping=1, unplantStiffness=250}
    }
    leftPinky {
     fK {
      enableFK=false
      poleVectorMaintainOffset=false
      poleVectorMatching=0
      poleVectorOffset=0
      rotationAlpha=1
      rotationMode=Interpolated
      translationAlpha=1
      translationMode=None
     }
     iK {
      bAffectedByIKWarping=true
      blendToSource=1
      blendToSourceRotation=0
      blendToSourceTranslation=1
      enableIK=false
      extension=1
      scaleVertical=1
      blendToSourceWeights=<1 1 1>
      staticLocalOffset=<0 0 0>
      staticOffset=<0 0 0>
      staticRotationOffset=<rot 0 0 0>
     }
     speedPlanting {enableSpeedPlanting=false, speedCurveName=None, speedThreshold=15, unplantCriticalDamping=1, unplantStiffness=250}
    }
    leftRing {
     fK {
      enableFK=false
      poleVectorMaintainOffset=false
      poleVectorMatching=0
      poleVectorOffset=0
      rotationAlpha=1
      rotationMode=Interpolated
      translationAlpha=1
      translationMode=None
     }
     iK {
      bAffectedByIKWarping=true
      blendToSource=1
      blendToSourceRotation=0
      blendToSourceTranslation=1
      enableIK=false
      extension=1
      scaleVertical=1
      blendToSourceWeights=<1 1 1>
      staticLocalOffset=<0 0 0>
      staticOffset=<0 0 0>
      staticRotationOffset=<rot 0 0 0>
     }
     speedPlanting {enableSpeedPlanting=false, speedCurveName=None, speedThreshold=15, unplantCriticalDamping=1, unplantStiffness=250}
    }
    rightIndex {
     fK {
      enableFK=false
      poleVectorMaintainOffset=false
      poleVectorMatching=0
      poleVectorOffset=0
      rotationAlpha=1
      rotationMode=Interpolated
      translationAlpha=1
      translationMode=None
     }
     iK {
      bAffectedByIKWarping=true
      blendToSource=1
      blendToSourceRotation=0
      blendToSourceTranslation=1
      enableIK=false
      extension=1
      scaleVertical=1
      blendToSourceWeights=<1 1 1>
      staticLocalOffset=<0 0 0>
      staticOffset=<0 0 0>
      staticRotationOffset=<rot 0 0 0>
     }
     speedPlanting {enableSpeedPlanting=false, speedCurveName=None, speedThreshold=15, unplantCriticalDamping=1, unplantStiffness=250}
    }
    rightMiddle {
     fK {
      enableFK=false
      poleVectorMaintainOffset=false
      poleVectorMatching=0
      poleVectorOffset=0
      rotationAlpha=1
      rotationMode=Interpolated
      translationAlpha=1
      translationMode=None
     }
     iK {
      bAffectedByIKWarping=true
      blendToSource=1
      blendToSourceRotation=0
      blendToSourceTranslation=1
      enableIK=false
      extension=1
      scaleVertical=1
      blendToSourceWeights=<1 1 1>
      staticLocalOffset=<0 0 0>
      staticOffset=<0 0 0>
      staticRotationOffset=<rot 0 0 0>
     }
     speedPlanting {enableSpeedPlanting=false, speedCurveName=None, speedThreshold=15, unplantCriticalDamping=1, unplantStiffness=250}
    }
    rightPinky {
     fK {
      enableFK=false
      poleVectorMaintainOffset=false
      poleVectorMatching=0
      poleVectorOffset=0
      rotationAlpha=1
      rotationMode=Interpolated
      translationAlpha=1
      translationMode=None
     }
     iK {
      bAffectedByIKWarping=true
      blendToSource=1
      blendToSourceRotation=0
      blendToSourceTranslation=1
      enableIK=false
      extension=1
      scaleVertical=1
      blendToSourceWeights=<1 1 1>
      staticLocalOffset=<0 0 0>
      staticOffset=<0 0 0>
      staticRotationOffset=<rot 0 0 0>
     }
     speedPlanting {enableSpeedPlanting=false, speedCurveName=None, speedThreshold=15, unplantCriticalDamping=1, unplantStiffness=250}
    }
    rightRing {
     fK {
      enableFK=false
      poleVectorMaintainOffset=false
      poleVectorMatching=0
      poleVectorOffset=0
      rotationAlpha=1
      rotationMode=Interpolated
      translationAlpha=1
      translationMode=None
     }
     iK {
      bAffectedByIKWarping=true
      blendToSource=1
      blendToSourceRotation=0
      blendToSourceTranslation=1
      enableIK=false
      extension=1
      scaleVertical=1
      blendToSourceWeights=<1 1 1>
      staticLocalOffset=<0 0 0>
      staticOffset=<0 0 0>
      staticRotationOffset=<rot 0 0 0>
     }
     speedPlanting {enableSpeedPlanting=false, speedCurveName=None, speedThreshold=15, unplantCriticalDamping=1, unplantStiffness=250}
    }
   }
   globalSettings {
    bCopyBasePose=false
    bEnableFK=true
    bEnableIK=true
    bEnablePost=true
    bEnableRoot=true
    bWarping=false
    copyBasePoseRoot=None
    directionChain=None
    directionSource=Goals
    forwardDirection=Y
    sidewaysOffset=0
    sourceScaleFactor=1
    warpForwards=1
    warpSplay=1
   }
   retargetOpProfiles []
   rootSettings {
    affectIKHorizontal=1
    affectIKVertical=0
    blendToSource=0
    rotationAlpha=1
    scaleHorizontal=1
    scaleVertical=1
    translationAlpha=1
    blendToSourceWeights=<1 1 1>
    rotationOffset=<rot 0 0 0>
    translationOffset=<0 0 0>
   }
  }
 }
 Output {type=SkeletalMeshComponentBudgeted}
 Retarget Profile {
  type=RetargetProfile
  default {
   bApplyChainSettings=false
   bApplyGlobalSettings=false
   bApplyRootSettings=false
   bApplySourceRetargetPose=false
   bApplyTargetRetargetPose=false
   bForceAllIKOff=false
   sourceRetargetPoseName=None
   targetRetargetPoseName=None
   chainSettings {}
   globalSettings {
    bCopyBasePose=false
    bEnableFK=true
    bEnableIK=true
    bEnablePost=true
    bEnableRoot=true
    bWarping=false
    copyBasePoseRoot=None
    directionChain=None
    directionSource=Goals
    forwardDirection=Y
    sidewaysOffset=0
    sourceScaleFactor=1
    warpForwards=1
    warpSplay=1
   }
   retargetOpProfiles []
   rootSettings {
    affectIKHorizontal=1
    affectIKVertical=0
    blendToSource=0
    rotationAlpha=1
    scaleHorizontal=1
    scaleVertical=1
    translationAlpha=1
    blendToSourceWeights=<1 1 1>
    rotationOffset=<rot 0 0 0>
    translationOffset=<0 0 0>
   }
  }
 }
]
